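# Unlit Modulate Niagara sprite material for the ClayPlug emote's FX, tinting scene surfaces like a flickering fire decal.
# With depth test off, it measures distance from the world position behind translucency to the particle center, forming a spherical fourth-power falloff scaled by particle size.
# A panning, stretched flame texture modulates that mask with particle alpha.
# The result blends from white, which is neutral under Modulate, toward particle color.
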Material /BRCosmetics/Animation/Game/MainPlayer/Emotes/ClayPlug/FX/M_ClayPlug_DecalFire
graph {
 properties {
  bDisableDepthTest=true
  bUsedWithNiagaraSprites=true
  blendMode=BLEND_Modulate
  pixelDepthOffsetMode=PDOM_Legacy
  shadingModel=MSM_Unlit
  translucencyPass=MTP_BeforeDOF
  usedShadingModels=Unlit
 }
 nodes [
  n1 = ParticleColor()
  n2 = MaterialFunctionCall(Float2=TextureCoordinate()) {materialFunction=&BreakOutFloat2Components}
  n3 = AppendVector(A=Multiply(A=@n2.R) {constB=0.2}, B=Multiply(A=@n2.G))
  n4 = TextureSampleParameter2D(UVs=Panner(Coordinate=Multiply(A=@n3) {constB=2}) {speedX=-0.1, speedY=-3}) {parameterName=FlameTexture, samplerType=SAMPLERTYPE_LinearGrayscale, texture=/Game/Effects/Fort_Effects/Textures/T_Emote_Monarch_Blast}
  n5 = LinearInterpolate(Alpha=@n4.R) {constA=0.2, constB=3}
  n6 = Distance(
   A=MaterialFunctionCall() {materialFunction=&WorldPositionBehindTranslucency}.Result
   B=ParticlePositionWS()
  )
  n7 = Multiply(A=ComponentMask(None=ParticleSize()) {r=true}) {constB=0.5}
  n8 = OneMinus(None=Saturate(None=Divide(A=@n6, B=@n7)))
  n9 = Multiply(A=@n8, B=@n8)
  LinearInterpolate(B=@n1.RGB, Alpha=Multiply(A=Multiply(A=@n1.A, B=@n5), B=Multiply(A=@n9, B=@n9))) {constA=1}
 ]
}
