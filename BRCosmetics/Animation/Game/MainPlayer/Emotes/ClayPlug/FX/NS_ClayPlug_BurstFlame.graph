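# Niagara system for the ClayPlug emote's flame burst, layering fire licks, SubUV fire, curl-noise sparks and a glow with a particle light.
# User.IsFlameBurst acts as spawn probability for the three fire emitters, while sparks and glow always spawn and use it to scale intensity.
# System FX_Scale, ToD and FadeOut values modulate size and brightness.
# Front-end modifiers dim the glow and sparks and gate the light.
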NiagaraSystem /BRCosmetics/Animation/Game/MainPlayer/Emotes/ClayPlug/FX/NS_ClayPlug_BurstFlame
dependencies {
 usedDataInterfaces [
  &NiagaraDataInterfaceCurve
  &NiagaraDataInterfaceColorCurve
  &NiagaraDataInterfaceSpriteRendererInfo
 ]
 usedDynamicInputs [&SimulationPosition]
 usedModules [
  &SystemState
  &self:SystemSpawnScript.SystemScriptSource.SystemScriptGraph.NiagaraNodeAssignment_3.SetVariables_#1
  &self:SystemSpawnScript.SystemScriptSource.SystemScriptGraph.NiagaraNodeAssignment_1.SetVariables_#2
  &self:SystemSpawnScript.SystemScriptSource.SystemScriptGraph.NiagaraNodeAssignment_2.SetVariables_#3
 ]
}
emitters {
 FireLicks_Sprites {
  emitterUpdateScript [
   &EmitterState()
   &SpawnBurst_Instantaneous(Spawn Count=9, Spawn Probability=$User.IsFlameBurst, Spawn Time=0.1, Use Spawn Probability=true, UseLoopCountLimit=false)
  ]
  particleSpawnScript [
   &InitializeParticle(
    Color=&RandomRangeLinearColor(Minimum=<rgba 0.8438 0.8438 0.8438 0.8>, Maximum=<rgba 1 1 1 1>)
    Lifetime=&RandomRangeFloat(Minimum=0.2, Maximum=0.7)
    Mass Mode=::NewEnumerator1
    Sprite Rotation Mode=::NewEnumerator2
    Sprite Size=&Vector2DFromFloat(Value=&Multiply_Float(
      A=&Multiply_Float(A=6, B=$System.FX_Scale)
      B=&UniformRangedFloat(Minimum=10, Maximum=30)
     ))
    Sprite Size Mode=::NewEnumerator4
    UsePositionOffset=false
    Write Material Random=true
   ) {module=InitializeParticle001}
   /SetVariables(Particles.StartingFrame:NiagaraInt32=&RandomRangeInt(Minimum=0, Maximum=32)) {isSetParameters=true}
   &Color(Color=<rgba 10 0.2 0 0.55>)
   &DynamicMaterialParameters(
    Index 0 Param 2=&RandomRangeFloat(Minimum=0, Maximum=4)
    Index 0 Param 4=2.5
    Module.Param0WriteEnabled=false
    Module.Param1WriteEnabled=true
    Module.Param2WriteEnabled=false
    Module.Param3WriteEnabled=true
   ) {module=DynamicMaterialParameters001}
   &ShapeLocation(
    MaskSpawnGroup=false
    Offset=<0 -10 5>
    Offset Coordinate Space=::Local
    Offset Mode=::NewEnumerator0
    Override Seed=false
    Sphere Radius=&Multiply_Float(A=$System.FX_Scale, B=50)
    UseOwnerScale=true
   )
   &InheritVelocity(Inherited Velocity Amount Scale=<0.1 0.1 0.1>, Inherited Velocity Speed Limit=250, Speed Limit=true, UseSpeedThreshold=false)
   &AddVelocity(
    Module.Use Falloff Distance=false
    Origin Offset=<0 -10 0>
    Override Seed=false
    Velocity Mode=::NewEnumerator1
    Velocity Origin=$StackContext.ShapeLocation.ShapePosition
    Velocity Speed=&Multiply_Float(A=&RandomRangeFloat(Minimum=1, Maximum=100), B=$System.FX_Scale)
   )
   &CameraOffset(Camera Offset Amount=15)
  ]
  particleUpdateScript [
   &ParticleState()
   &ScaleColor(
    Scale Alpha=&Multiply_Float(
     A=0.8
     B=&FloatFromCurve(
      FloatCurve=DataInterface() {propertyValues {Curve {keys [
          {arriveTangent=0.0403, interpMode=RCIM_Cubic, leaveTangent=0.0403, tangentMode=RCTM_User, time=0, value=0.1}
          {arriveTangent=-0.0116, interpMode=RCIM_Cubic, leaveTangent=-0.0116, tangentMode=RCTM_User, time=0.2, value=1}
          {interpMode=RCIM_Cubic, leaveTangent=-2.371, time=1, value=0}
         ]}}}
      Scale Curve=$System.ToD
     )
    )
    Scale RGB=&MakeVectorFromLinearColorRGB(Color=&ColorFromCurve(ColorCurve=DataInterface() {propertyValues {
        AlphaCurve {keys [<at 0 1 RCIM_Cubic>]}
        BlueCurve {keys [<at 0.1464 1.6881 RCIM_Cubic>,<at 0.8987 0 RCIM_Cubic>]}
        GreenCurve {keys [<at 0.1464 6.1406 RCIM_Cubic>,<at 0.8987 0 RCIM_Cubic>]}
        RedCurve {keys [<at 0.1464 8 RCIM_Cubic>,<at 0.8987 0.1 RCIM_Cubic>]}
       }}))
    ScaleA=true
    ScaleRGB=true
   ) {module=ScaleColor001}
   &ScaleSpriteSize(Scale Factor=&Vector2DFromFloat(Value=&FloatFromCurve(FloatCurve=DataInterface() {propertyValues {Curve {keys [{interpMode=RCIM_Cubic, tangentMode=RCTM_User, time=0, value=0.3}, {interpMode=RCIM_Cubic, tangentMode=RCTM_User, time=1, value=1}]}}})))
   &DynamicMaterialParameters(
    Index 0 Param 1=&FloatFromCurve(FloatCurve=DataInterface() {propertyValues {Curve {keys [<at 0 1>]}}})
    Index 0 Param 2=0.5
    Index 0 Param 3=&FloatFromCurve(FloatCurve=DataInterface() {propertyValues {Curve {keys [<at 0 0.5 RCIM_Cubic>, {arriveTangent=1.5, interpMode=RCIM_Cubic, leaveTangent=1.5, time=0.5235, value=0.5}, <at 1 2>]}}})
    Module.Param0WriteEnabled=true
    Module.Param1WriteEnabled=false
    Module.Param2WriteEnabled=true
    Module.Param3WriteEnabled=false
   )
   &Drag(Drag=5, Use Linear Drag=true, Use Rotational Drag=true)
   &AccelerationForce(Acceleration=&RandomRangeVector(Minimum=<0 0 50>, Maximum=<0 0 300>))
   &InheritSourceMovement(Applied Position Scale=<0.2 0.2 0.2>, ApplyForce=false, ApplyPosition=true, ApplyPositionFalloff=false)
   &SolveForcesAndVelocity(Clamp Velocity=false, Limit Acceleration=false)
   &SubUVAnimation(
    End Frame Range Override:NiagaraInt32=&Add_Integer(A=$Particles.StartingFrame, B=32)
    "Sprite Renderer "=DataInterface() {propertyValues {SpriteRenderer=&self:FireLicks_Sprites_0.NiagaraSpriteRendererProperties_0}}
    Start Frame Offset:NiagaraInt32=$Particles.StartingFrame
    UseEndFrame=true
    UseStartFrame=false
   )
  ]
  rendererClasses [&NiagaraSpriteRendererProperties]
  renderers [{rendererClass=&NiagaraSpriteRendererProperties}]
 }
 FireLicks_Sprites_Tall001 {
  emitterUpdateScript [
   &EmitterState(Enable Distance Culling=true, MaxDistance=3000, Module.bUseMaxDistance=true, Module.bUseMinDistance=false, Scalability Mode=::NewEnumerator1)
   &SpawnBurst_Instantaneous(Spawn Count=8, Spawn Probability=$User.IsFlameBurst, Use Spawn Probability=true, UseLoopCountLimit=false)
  ]
  particleSpawnScript [
   &InitializeParticle(
    Color=<rgba 1 1 1 0.333>
    Lifetime Max=0.75
    Lifetime Min=0.5
    Lifetime Mode=::NewEnumerator1
    Mass Mode=::NewEnumerator1
    Position Offset=<0 -7.5 0>
    Position Offset Coordinate Space=::Local
    Sprite Rotation Mode=::NewEnumerator2
    Sprite Size Mode=::NewEnumerator1
    UsePositionOffset=true
    Write Material Random=true
   ) {module=InitializeParticle001}
   &DynamicMaterialParameters(
    Index 0 Param 2=&RandomRangeFloat(Minimum=0, Maximum=4)
    Index 0 Param 4=2.5
    Module.Param0WriteEnabled=false
    Module.Param1WriteEnabled=true
    Module.Param2WriteEnabled=false
    Module.Param3WriteEnabled=true
   ) {module=DynamicMaterialParameters001}
   &ShapeLocation(MaskSpawnGroup=false, Non Uniform Scale=<1 1 2>, Override Seed=false, Sphere Radius=5, UseOwnerScale=true)
   &AddVelocity(
    Module.Use Falloff Distance=false
    Origin Offset=<0 -7.5 -10>
    Override Seed=false
    Velocity Mode=::NewEnumerator1
    Velocity Speed=&RandomRangeFloat(Minimum=25, Maximum=50)
   )
  ]
  particleUpdateScript [
   &ParticleState()
   &ScaleSpriteSize(Scale Factor=&Vector2DFromFloat(Value=&FloatFromCurve(
      FloatCurve=DataInterface() {propertyValues {Curve {keys [
          {arriveTangent=7.5764, interpMode=RCIM_Cubic, leaveTangent=7.5765, tangentMode=RCTM_User, time=0, value=0.02}
          {arriveTangent=0.0102, interpMode=RCIM_Cubic, leaveTangent=0.0102, tangentMode=RCTM_User, time=0.18, value=0.99}
          <at 1 0.1 RCIM_Cubic>
         ]}}}
      Scale Curve=1.2
     ))) {module=ScaleSpriteSize001}
   &AccelerationForce(Acceleration=&RandomRangeVector(Minimum=<0 0 50>, Maximum=<0 0 300>))
   &DynamicMaterialParameters(
    Index 0 Param 1=&FloatFromCurve(FloatCurve=DataInterface() {propertyValues {Curve {keys [<at 0 1>]}}})
    Index 0 Param 2=0.5
    Index 0 Param 3=&FloatFromCurve(FloatCurve=DataInterface() {propertyValues {Curve {keys [
         <at 0 0.5 RCIM_Cubic>
         {arriveTangent=-0.0025, interpMode=RCIM_Cubic, leaveTangent=-0.0025, tangentMode=RCTM_User, time=0.6, value=0.625}
         {arriveTangent=2.7303, tangentMode=RCTM_User, time=1, value=2}
        ]}}})
    Module.Param0WriteEnabled=true
    Module.Param1WriteEnabled=false
    Module.Param2WriteEnabled=true
    Module.Param3WriteEnabled=false
   )
   &SpriteRotationRate(Rotation Rate=&Lerp_Float(A=-10, B=10, Alpha=$Particles.Initial.MaterialRandom)) {module=SpriteRotationRate001}
   /SetVariables(Particles.SubImageIndex=&FloatFromCurve(FloatCurve=DataInterface() {propertyValues {Curve {keys [<at 0 5>,<at 1 15>]}}})) {isSetParameters=true}
   &SolveForcesAndVelocity(Clamp Velocity=false, Limit Acceleration=false)
   &CameraOffset(Camera Offset Amount=10)
   &ScaleColor(
    Scale Alpha=&Multiply_Float(
     A=1
     B=&FloatFromCurve(
      FloatCurve=DataInterface() {propertyValues {Curve {keys [
          {arriveTangent=0.9552, interpMode=RCIM_Cubic, leaveTangent=0.9552, tangentMode=RCTM_User, time=0, value=0}
          {arriveTangent=-2.3981, interpMode=RCIM_Cubic, leaveTangent=-2.3981, tangentMode=RCTM_User, time=0.1, value=0.95}
          {arriveTangent=-1.2175, interpMode=RCIM_Cubic, leaveTangent=-1.2175, tangentMode=RCTM_User, time=0.3, value=0.2}
          {arriveTangent=-0.3081, interpMode=RCIM_Cubic, leaveTangent=-0.3081, tangentMode=RCTM_User, time=1, value=0}
         ]}}}
      Scale Curve=$System.FadeOut
     )
    )
    Scale RGB=&MakeVectorFromLinearColorRGB(Color=&ColorFromCurve(ColorCurve=DataInterface() {propertyValues {
        AlphaCurve {keys [<at 0.0141 0>,<at 0.2566 0.8>,<at 0.6705 0.5>,<at 1 0 RCIM_Cubic>]}
        BlueCurve {keys [<at 0 0 RCIM_Cubic>,<at 0.1098 0>,<at 0.3438 0>,<at 0.5257 0>,<at 0.6631 0.16>,<at 0.7992 0>,<at 1 0>]}
        GreenCurve {keys [<at 0 0 RCIM_Cubic>,<at 0.1098 0.1167>,<at 0.3438 0.4667>,<at 0.5257 2.5667>,<at 0.6631 5.176>,<at 0.7992 0.3667>,<at 1 0>]}
        RedCurve {keys [<at 0 1 RCIM_Cubic>,<at 0.1098 1>,<at 0.3438 4>,<at 0.5257 11>,<at 0.6631 16>,<at 0.7992 1>,<at 1 1>]}
       }}))
    ScaleA=true
    ScaleRGB=true
   )
  ]
  rendererClasses [&NiagaraSpriteRendererProperties]
  renderers [{rendererClass=&NiagaraSpriteRendererProperties}]
 }
 Fire_Sprites_SubUV {
  emitterUpdateScript [
   &EmitterState()
   &SpawnBurst_Instantaneous(Spawn Probability=$User.IsFlameBurst, Use Spawn Probability=true, UseLoopCountLimit=false)
  ]
  particleSpawnScript [
   &InitializeParticle(
    Color=<rgba 1 1 1 0.5>
    Lifetime=&UniformRangedFloat(Minimum=0.5, Maximum=0.75)
    Mass Mode=::NewEnumerator1
    Position Offset=<0 -7.5 0>
    Sprite Rotation Angle=&UniformRangedFloat(Minimum=-40, Maximum=40)
    Sprite Rotation Mode=::NewEnumerator0
    Sprite Size=&Vector2DFromFloat(Value=&UniformRangedFloat(Minimum=15, Maximum=25))
    Sprite Size Mode=::NewEnumerator4
    UsePositionOffset=true
    Write Material Random=true
   ) {module=InitializeParticle001}
   &ShapeLocation(
    MaskSpawnGroup=false
    Non Uniform Scale=<1 1 10>
    Offset=<5 0 0>
    Offset Coordinate Space=::Local
    Offset Mode=::NewEnumerator0
    Override Seed=false
    Sphere Radius=1
    UseOwnerScale=true
   ) {module=ShapeLocation001}
  ]
  particleUpdateScript [
   &ParticleState()
   &SubUVAnimation(SubUV Lookup Index:NiagaraFloat=$Engine.System.Age)
   &SolveForcesAndVelocity(Clamp Velocity=false, Limit Acceleration=false)
   &ScaleColor(
    Scale Alpha=&Multiply_Float(
     A=&FloatFromCurve(
      FloatCurve=DataInterface() {propertyValues {Curve {keys [
          {arriveTangent=16.5612, interpMode=RCIM_Cubic, leaveTangent=16.5612, leaveTangentWeight=0.3333, tangentMode=RCTM_User, time=0, value=0}
          {
           arriveTangent=-3.594
           arriveTangentWeight=0.0333
           interpMode=RCIM_Cubic
           leaveTangent=-3.594
           leaveTangentWeight=0.3
           tangentMode=RCTM_User
           time=0.1
           value=1
          }
          {arriveTangentWeight=0.1667, interpMode=RCIM_Cubic, tangentMode=RCTM_User, time=1, value=0}
         ]}}}
      Scale Curve=$System.FadeOut
     )
     B=&RandomRangeFloat(Minimum=0.5, Maximum=1)
    )
    Scale RGB=&Multiply_VectorByFloat(
     Vector=<2.5 2.5 2.5>
     Float=&FloatFromCurve(
      FloatCurve=DataInterface() {propertyValues {Curve {keys [
          {arriveTangent=9.0725, interpMode=RCIM_Cubic, leaveTangent=9.0725, tangentMode=RCTM_User, time=0, value=0}
          {arriveTangent=-2.5249, interpMode=RCIM_Cubic, leaveTangent=-2.5249, tangentMode=RCTM_User, time=0.1408, value=0.9847}
          {arriveTangent=-0.0405, interpMode=RCIM_Cubic, leaveTangent=-0.0405, tangentMode=RCTM_User, time=1, value=0}
         ]}}}
      Scale Curve=&RandomRangeFloat(Minimum=0.5, Maximum=1)
     )
    )
    ScaleA=true
    ScaleRGB=true
   )
   &ScaleSpriteSize(Scale Factor=&Vector2DFromFloat(Value=&FloatFromCurve(FloatCurve=DataInterface() {propertyValues {Curve {keys [
          <at 0 0.7 RCIM_Cubic>
          {arriveTangent=-0.565, interpMode=RCIM_Cubic, leaveTangent=-0.565, tangentMode=RCTM_User, time=0.2, value=1.5}
          {interpMode=RCIM_Cubic, tangentMode=RCTM_User, time=1, value=0.75}
         ]}}}))) {module=ScaleSpriteSize001}
   &DynamicMaterialParameters(
    Index 0 Param 1=&RandomRangeFloat(Minimum=2, Maximum=1)
    Index 0 Param 2=&RandomRangeFloat(Minimum=0, Maximum=2)
    Index 0 Param 3=8
    Index 0 Param 4=$Particles.SubImageIndex
    Index 1 Param 1=&RandomRangeFloat(Minimum=-1, Maximum=1)
    Index 1 Param 2=&RandomRangeFloat(Minimum=-1, Maximum=1)
    Index 1 Param 3=&FloatFromCurve(FloatCurve=DataInterface() {propertyValues {Curve {keys [
         {arriveTangent=4.1785, interpMode=RCIM_Cubic, leaveTangent=4.1785, tangentMode=RCTM_User, time=0, value=0}
         {arriveTangent=0.9995, interpMode=RCIM_Cubic, leaveTangent=0.9995, tangentMode=RCTM_User, time=0.2223, value=0.6825}
         {arriveTangent=-0.0405, interpMode=RCIM_Cubic, leaveTangent=-0.0405, tangentMode=RCTM_User, time=0.7, value=0.9}
         <at 1 0.7>
        ]}}})
    Index 2 Param 4=1
    Module.Param0WriteEnabled=true
    Module.Param0WriteEnabled1=true
    Module.Param0WriteEnabled2=false
    Module.Param1WriteEnabled=true
    Module.Param1WriteEnabled1=true
    Module.Param1WriteEnabled2=false
    Module.Param2WriteEnabled=true
    Module.Param2WriteEnabled1=true
    Module.Param2WriteEnabled2=false
    Module.Param3WriteEnabled=true
    Module.Param3WriteEnabled1=false
    Module.Param3WriteEnabled2=true
    Write Parameter Index 1=true
    Write Parameter Index 2=true
   )
   /SetVariables(Particles.Velocity=&TransformVector(Vector=<0 0 1>, Source Space=::World, Destination Space=::Local)) {isSetParameters=true}
   &CameraOffset(Camera Offset Amount=20)
  ]
  rendererClasses [&NiagaraSpriteRendererProperties]
  renderers [{rendererClass=&NiagaraSpriteRendererProperties}]
 }
 Glow_Light {
  emitterUpdateScript [
   &EmitterState()
   &SpawnBurst_Instantaneous(Spawn Group=1, Use Spawn Probability=false, UseLoopCountLimit=false) {module=SpawnBurst_Instantaneous001}
   &SpawnBurst_Instantaneous(Spawn Probability=$User.IsFlameBurst, Use Spawn Probability=true, UseLoopCountLimit=false)
  ]
  particleSpawnScript [
   /SetVariables(Particles.VisibilityTag=&SpawnGroup()) {isSetParameters=true}
   &InitializeParticle(
    Color=<rgba 50 11.3817 0.7 1>
    Lifetime=5
    Position Offset=<0 -7.5 0>
    Sprite Rotation Mode=::NewEnumerator2
    Sprite Size=<50 50>
    Sprite Size Mode=::NewEnumerator4
    UsePositionOffset=true
    Write Material Random=true
   )
   &AddVelocity(Velocity=<0.1 0 0>)
   &CameraOffset(Camera Offset Amount=2)
  ]
  particleUpdateScript [
   &SolveForcesAndVelocity(Clamp Velocity=false, Limit Acceleration=false)
   &ParticleState()
   &ScaleColor(
    Scale Alpha=&Multiply_Float(
     A=&Waveform(
      Global Input Position=&RandomRangeFloat(Minimum=0, Maximum=1)
      "Waveform [1]"=::NewEnumerator0
      "[1] Frequency"=1
      "[1] Amplitude Min/Max"=<0.5 1>
      "[1] Scale"=&Waveform(
       Waveform Count=::NewEnumerator1
       Global Input Position=0.33
       "Waveform [1]"=::NewEnumerator0
       "[1] Frequency"=0.33
       "[1] Amplitude Min/Max"=<0.5 1.5>
       "[2] Average"=false
       "[2] Input Position"=0
       "[2] Phase"=$Emitter.Age
       "[2] Phase Scale"=1
       "[2] Compound Frequency"=2
       "[2] Frequency"=2
       "[2] Amplitude Min/Max"=<0 1>
       "[2] Scale"=1
       "[2] Bias"=0
       "[2] Exponent"=1
       "[2] Pulse Duration"=0.1
       "[2] Clip Triangle Min/Max"=<0 0>
       "[2] Random Seed"=0
      )
     )
     B=&Multiply_Float(A=$System.FadeOut, B=$System.ToD)
    )
    Scale RGB=<0.5 0.5 0.5>
    ScaleA=true
    ScaleRGB=true
   )
   /SetVariables(Particles.ColorGlow:LinearColor=&MultiplyLinearColor(
     Linear Color=$Particles.Color
     Scale Factor=&MakeVector4(
      X=1
      Y=1
      Z=1
      W=&Multiply_Float(
       A=0.005
       B=&Multiply_Float(
        A=$System.ToD
        B=&FN_FrontEndFloatModifier(Gameplay=&Lerp_Float(A=0.75, B=1, Alpha=$User.IsFlameBurst), Front End=0.5)
       )
      )
     )
    )) {isSetParameters=true}
   /SetVariables(Particles.LightColor=&MultiplyLinearColor(
     Linear Color=$Particles.Color
     Scale Factor=&MakeFloat4FromVectorAndFloat(
      "Vector (RGB)"=<1 1 1>
      "Float (Alpha)"=&FNTimeOfDayScalar(Night Scale=0.5, Morning Scale=1, Evening Scale=1)
     )
    )) {isSetParameters=true}
   &FN_Light_Attributes(
    Color=<rgba 1 0.2667 0 1>
    Create Color Attribute=true
    Create Exponent Attribute=true
    Create Radius Attribute=true
    Enable Particle Light Enable=true
    Exponent=3
    Light Enabled:NiagaraBool=&SetBoolByFloatComparison(A=&FN_FrontEndFloatModifier(Gameplay=1, Front End=0), B=1, Comparison Type=::NewEnumerator4)
    Radius=30
    Scale Color=&Multiply_Float(
     A=7
     B=&FloatFromCurve(
      FloatCurve=DataInterface() {propertyValues {Curve {keys [
          {interpMode=RCIM_Cubic, tangentMode=RCTM_User, time=0, value=0}
          {interpMode=RCIM_Cubic, tangentMode=RCTM_User, time=0.5, value=2}
          {tangentMode=RCTM_User, time=1, value=0}
         ]}}}
      Scale Curve=&Multiply_Float(A=$System.FadeOut, B=$System.ToD)
     )
    )
   )
  ]
  rendererClasses [&NiagaraSpriteRendererProperties, &NiagaraLightRendererProperties]
  renderers [
   {rendererClass=&NiagaraSpriteRendererProperties}
   {rendererClass=&NiagaraLightRendererProperties, rendererIndex=1}
   {rendererClass=&NiagaraSpriteRendererProperties, rendererIndex=2}
  ]
 }
 Sparks {
  emitterUpdateScript [
   &EmitterState()
   &SpawnRate(
    SpawnRate=&FloatFromCurve(
     FloatCurve=DataInterface() {propertyValues {Curve {keys [
         {arriveTangent=9.0725, interpMode=RCIM_Cubic, leaveTangent=9.0725, tangentMode=RCTM_User, time=0, value=1}
         {arriveTangent=-4.9128, interpMode=RCIM_Cubic, leaveTangent=-4.9128, tangentMode=RCTM_User, time=0.1, value=0.225}
         {arriveTangent=-0.0405, interpMode=RCIM_Cubic, leaveTangent=-0.0405, tangentMode=RCTM_User, time=0.2, value=0}
        ]}}}
     CurveIndex=$Emitter.NormalizedLoopAge
     Scale Curve=&Lerp_Float(A=25, B=50, Alpha=$User.IsFlameBurst)
    )
    Use Spawn Probability=false
   )
   &SpawnBurst_Instantaneous(
    Spawn Count=&MakeIntFromFloat(Float=&Lerp_Float(A=15, B=17, Alpha=$User.IsFlameBurst))
    Use Spawn Probability=false
    UseLoopCountLimit=false
   )
  ]
  particleSpawnScript [
   &InitializeParticle(
    Color=&LerpLinearColors(
     StartColor=<rgba 10 1.0866 0 1>
     EndColor=<rgba 2 0.0333 0 1>
     LerpFactor=&UniformAOrBFloat(A=0, Distribution Weight=0.2, Evaluation Type=::NewEnumerator0, Recalculate Random Each Loop=true, Override Seed=false)
    )
    Lifetime Max=&Lerp_Float(A=1, B=1.5, Alpha=$User.IsFlameBurst)
    Lifetime Min=0.3
    Lifetime Mode=::NewEnumerator1
    Position Offset=<0 -7.5 0>
    Position Offset Coordinate Space=::Local
    Sprite Size Mode=::NewEnumerator1
    Uniform Sprite Size Max=2
    Uniform Sprite Size Min=0.8
    UsePositionOffset=true
    Write Material Random=true
   )
   &ShapeLocation(
    Cylinder Height=10
    Cylinder Radius=6
    MaskSpawnGroup=false
    Offset=<7 0 0>
    Offset Coordinate Space=::Local
    Offset Mode=::NewEnumerator0
    Override Seed=false
    Rotation Mode=::NewEnumerator1
    Shape Primitive=::NewEnumerator1
    Surface Only=false
    UseOwnerScale=true
   )
   &AddVelocityFromPoint(
    Module.Use Falloff Distance=false
    Origin Offset=&Lerp_Vector(A=<7 -7.5 -5>, B=&RandomRangeVector(Minimum=<7 -7.5 0>, Maximum=<7 -7.5 -10>), Alpha=0.5)
    Origin Offset Coordinate Space=::World
    Velocity Strength=&Lerp_Float(A=25, B=&RandomRangeFloat(Minimum=50, Maximum=250), Alpha=$User.IsFlameBurst)
   )
   &InheritVelocity(Inherited Velocity Amount Scale=<0.9 0.9 0.9>, Inherited Velocity Speed Limit=350, Speed Limit=true, UseSpeedThreshold=false)
  ]
  particleUpdateScript [
   &ParticleState()
   &AccelerationForce(Acceleration=&Multiply_Vector(
     A=&VectorFromFloat(Value=0.5)
     B=&UniformRangedVector(Minimum=<0 0 150>, Maximum=<0 0 250>)
    ))
   &Drag(
    Drag=&FloatFromCurve(FloatCurve=DataInterface() {propertyValues {Curve {keys [<at 0 0.5>,<at 0.5 0.5>,<at 1 1>]}}}, Scale Curve=12)
    Use Linear Drag=true
    Use Rotational Drag=false
   )
   /Niagara/Modules/Update/Forces/CurlNoiseForce(
    Noise Frequency:NiagaraFloat=&RandomRangeFloat(Minimum=0.05, Maximum=0.3)
    Noise Strength=&FloatFromCurve(FloatCurve=DataInterface() {propertyValues {Curve {keys [<at 0 1>,<at 1 0>]}}}, Scale Curve=500)
    Pan Noise=true
    Pan Noise Field=<0 0 1>
   )
   &ScaleSpriteSize(Scale Factor=&Vector2DFromFloat(Value=&FloatFromCurve(FloatCurve=DataInterface() {propertyValues {Curve {keys [
          {arriveTangent=9.1531, interpMode=RCIM_Cubic, leaveTangent=9.1531, tangentMode=RCTM_User, time=0, value=0.02}
          {arriveTangent=-0.02, interpMode=RCIM_Cubic, leaveTangent=-0.02, time=0.18, value=0.99}
          <at 1 0 RCIM_Cubic>
         ]}}}))) {module=ScaleSpriteSize002}
   &SolveForcesAndVelocity(Clamp Velocity=false, Limit Acceleration=false)
   &ScaleColor(
    Scale Alpha=&FloatFromCurve(
     FloatCurve=DataInterface() {propertyValues {Curve {keys [
         {arriveTangent=9.1531, interpMode=RCIM_Cubic, leaveTangent=9.1531, tangentMode=RCTM_User, time=0, value=0.02}
         {arriveTangent=0.48, interpMode=RCIM_Cubic, leaveTangent=0.48, time=0.18, value=0.99}
         <at 1 0.5 RCIM_Cubic>
        ]}}}
     Scale Curve=&MakeCustomFloatFromBool(Bool=$User.bIsFrontEndPreview, True Float=0.5, False Float=1)
    )
    Scale RGB=&MakeVectorFromLinearColorRGB(Color=&ColorFromCurve(ColorCurve=DataInterface() {propertyValues {
        AlphaCurve {keys [<at 0 1 RCIM_Cubic>]}
        BlueCurve {keys [<at 0 1.2255 RCIM_Cubic>,<at 0.6079 0 RCIM_Cubic>]}
        GreenCurve {keys [<at 0 3.7041 RCIM_Cubic>,<at 0.6079 0.0133 RCIM_Cubic>]}
        RedCurve {keys [<at 0 5 RCIM_Cubic>,<at 0.6079 0.8 RCIM_Cubic>]}
       }}))
    ScaleA=true
    ScaleRGB=true
   ) {module=ScaleColor001}
   &ScaleSpriteSizeBySpeed(
    Max Scale Factor=<1.5 5>
    Min Scale Factor=<0.4 0.8>
    Sample Scale Factor By Curve=true
    Scale Factor Curve=DataInterface() {propertyValues {Curve {keys [
        <at -0.0077 0.0765 RCIM_Cubic>
        {arriveTangent=1.2463, interpMode=RCIM_Cubic, leaveTangent=1.2463, tangentMode=RCTM_User, time=0.9985, value=0.9549}
       ]}}}
    Velocity Threshold=500
   ) {module=ScaleSpriteSizeBySpeed001}
   &FN_Sprite_Attributes(
    Color=&MultiplyLinearColor(Linear Color=$Particles.Color, Scale Factor=<1 1 1 0.005>)
    Create Color Attribute=true
    Create SpriteSize Attribute=true
    SpriteSize=&Multiply_Vector2D_ByFloat(Float=5, Vector2D=$Particles.SpriteSize)
    Write to Module Namespace=true
   )
  ]
  rendererClasses [&NiagaraSpriteRendererProperties]
  renderers [
   {rendererClass=&NiagaraSpriteRendererProperties}
   {rendererClass=&NiagaraSpriteRendererProperties, rendererIndex=1}
  ]
 }
}
summary {
 systemName=NS_ClayPlug_BurstFlame
 emitters [
  {emitterName=Glow_Light, rendererClasses [&NiagaraSpriteRendererProperties, &NiagaraLightRendererProperties]}
  {emitterName=Sparks, rendererClasses [&NiagaraSpriteRendererProperties]}
  {emitterName=FireLicks_Sprites_Tall001, rendererClasses [&NiagaraSpriteRendererProperties]}
  {emitterName=Fire_Sprites_SubUV, rendererClasses [&NiagaraSpriteRendererProperties]}
  {emitterName=FireLicks_Sprites, rendererClasses [&NiagaraSpriteRendererProperties]}
 ]
 userVariables [
  User.bIsFrontEnd {type=NiagaraBool}
  User.bIsFrontEndPreview {type=NiagaraBool}
  User.IsFlameBurst {type=NiagaraFloat, defaultValue=0, description=""}
  User.Reveal {type=NiagaraFloat, defaultValue=0, description=""}
  User.RibbonIDOffset {type=NiagaraInt32}
 ]
}
